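; B_RA_FXIndicator is a self-destroying Blueprint actor that plays a Niagara indicator effect for a Rift Anomaly.
; The construction script shifts it by a spawn offset, and BeginPlay activates the effect and debug-prints its location.
; A timeline then ramps the RiftAnomalyActive scalar in the shared FortniteMaterialParameters collection, affecting materials globally.
; The actor destroys itself when the Niagara system finishes, without resetting that parameter.
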(event EventBeginPlay
  (bind _niagara (Variables|Default|GetNiagara))
  (Components|Activation|Activate _niagara false)
  (EventDispatchers|BindEventtoOnSystemFinished _niagara (EventDispatchers|CreateEvent :!event Custom|OnSystemFinished))
  (Debugging|PrintVector (Transformation|GetActorLocation) "Rift Anomaly FX loc = " 25.0 "(R=0.920714,G=0.241936,B=1.000000,A=1.000000)"
    (:Then
      ;[1 1 1] MPC RiftAnomalyActive 0-1 over 1 Sec
      (bind (_direction _riftanomalyactive) (Timeline|Timeline Play)
        (:Update
          (Rendering|Material|SetScalarParameterValue "/Game/Packages/Fortress_SharedMaterials/GlobalMaterialParameters/FortniteMaterialParameters.FortniteMaterialParameters" "RiftAnomalyActive" _riftanomalyactive :!class /Script/Engine.KismetMaterialLibrary))))))

(event Custom|OnSystemFinished
  (params (PSystem :type Niagara%20Particle%20System%20Component%20Object%20Reference))
  (Actor|DestroyActor))

(fn ConstructionScript
  (Transformation|AddActorWorldOffset (Variables|Default|GetSpawnOffset)))
