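# Translucent unlit beam material for Niagara sprites and ribbons, named for a match modifier effect.
# A slowly panning Perlin noise sample is multiplied by a V × (1−V) gradient that fades both ends of the V axis, then shaped by a power and boosted.
# The result is multiplied by a view-angle fade with tunable sensitivity and a depth fade to soften edge-on views and intersections, alongside a particle color input for tinting.
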Material /RiftAnomaliesGameplay/Assets/FX/M_MatchModifier_Beam
graph {
 properties {
  bAutomaticallySetUsageInEditor=false
  bUsedWithNiagaraRibbons=true
  bUsedWithNiagaraSprites=true
  blendMode=BLEND_Translucent
  shadingModel=MSM_Unlit
  usedShadingModels=Unlit
 }
 nodes [
  n1 = Panner(Coordinate=TextureCoordinate() {vTiling=0.1}) {speedY=0.1}
  n2 = TextureCoordinate()
  n3 = ComponentMask(None=@n2) {g=true}
  n4 = OneMinus(None=ComponentMask(None=@n2) {g=true})
  n5 = Power(Base=Multiply(
    A=TextureSample(UVs=@n1) {samplerType=SAMPLERTYPE_Masks, texture=/Game/Global/Textures/Default/Noise/T_Generic_Perlin}.R
    B=Multiply(A=@n3, B=@n4)
   ))
  n6 = ScalarParameter() {defaultValue=1, parameterName=AngleFade Sensitivity}
  Multiply(
   A=Multiply(
    A=Multiply(A=@n5) {constB=10}
    B=MaterialFunctionCall("FadeSensitivity (optional)"=@n6) {materialFunction=&FadeBasedOnViewAngle}.Result
   )
   B=DepthFade()
  )
  ParticleColor()
 ]
}
