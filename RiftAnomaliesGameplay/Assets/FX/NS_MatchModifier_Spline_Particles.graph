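# Niagara system for a match modifier effect in Rift Anomalies that draws colored ribbons and flowing particles along a spline.
# Two ribbon beam emitters each span half the spline, taking endpoints and tangents from a scratch module.
# Dense GPU and sparser CPU sprites spawn around the spline, driven by curl noise and custom SplinePhysics magnet, forward and spiral forces.
# Sprites stretch with speed and fade over life; the system exposes a Color user parameter.
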NiagaraSystem /RiftAnomaliesGameplay/Assets/FX/NS_MatchModifier_Spline_Particles
dependencies {
 usedDataInterfaces [&NiagaraDataInterfaceCurve, /Script/Niagara.NiagaraDataInterfaceSpline]
 usedDynamicInputs []
 usedModules [
  &SystemState
  &self:SystemSpawnScript.SystemScriptSource.SystemScriptGraph.NiagaraNodeAssignment_0.SetVariables_#1
  &self:SystemSpawnScript.SystemScriptSource.SystemScriptGraph.NiagaraNodeAssignment_1.SetVariables_#2
  &self:SystemSpawnScript.SystemScriptSource.SystemScriptGraph.NiagaraNodeAssignment_2.SetVariables_#3
 ]
}
emitters {
 Beam_A {
  emitterUpdateScript [
   &EmitterState(Life Cycle Mode=::NewEnumerator1, Loop Behavior=::NewEnumerator1, UseLoopDelay=false)
   &self:ScratchModule(End U=0.5, Spline:NiagaraDataInterfaceSpline=$System.Spline, Tangent End U=0.5) {module=ScratchModule}
   /Niagara/Modules/Beams/BeamEmitterSetup(
    Absolute Beam End=true
    Beam End:NiagaraPosition=$Emitter.End
    Beam End Tangent:Vector3f=&Multiply_VectorByFloat(Vector=$Emitter.EndTangent, Float=-0.66)
    Beam Start:NiagaraPosition=$Emitter.Start
    Beam Start Tangent:Vector3f=&Multiply_VectorByFloat(Vector=$Emitter.StartTangent, Float=0.66)
    Use Beam Tangents=true
   )
   &SpawnBurst_Instantaneous(Spawn Count=30, Use Spawn Probability=false, UseLoopCountLimit=false)
  ]
  particleSpawnScript [
   &InitializeParticle(Color=$System.Color, Lifetime=5, Position Mode=::NewEnumerator0, Write Material Random=true)
   /Niagara/Modules/Beams/SpawnBeam()
   /Niagara/Modules/Beams/BeamWidth(Beam Width=1000)
   /SetVariables() {isSetParameters=true}
  ]
  particleUpdateScript [
   &ParticleState(Kill Particles When Lifetime Has Elapsed=false, Loop Particles Lifetime=true)
   &SolveForcesAndVelocity(Clamp Velocity=false, Limit Acceleration=false)
   /Niagara/Modules/Beams/UpdateBeam()
  ]
  rendererClasses [&NiagaraRibbonRendererProperties]
  renderers [{rendererClass=&NiagaraRibbonRendererProperties}]
 }
 Beam_B {
  emitterUpdateScript [
   &EmitterState(Life Cycle Mode=::NewEnumerator1, Loop Behavior=::NewEnumerator1, UseLoopDelay=false)
   &self:ScratchModule(End U=1, Spline:NiagaraDataInterfaceSpline=$System.Spline, Start U=0.5, Tangent End U=1, Tangent Start U=0.5) {module=ScratchModule}
   &BeamEmitterSetup(
    Absolute Beam End=true
    Beam End:NiagaraPosition=$Emitter.End
    Beam End Tangent:Vector3f=&Multiply_VectorByFloat(Vector=$Emitter.EndTangent, Float=-0.66)
    Beam Start:NiagaraPosition=$Emitter.Start
    Beam Start Tangent:Vector3f=&Multiply_VectorByFloat(Vector=$Emitter.StartTangent, Float=0.66)
    Use Beam Tangents=true
   )
   &SpawnBurst_Instantaneous(Spawn Count=30, Use Spawn Probability=false, UseLoopCountLimit=false)
  ]
  particleSpawnScript [
   &InitializeParticle(Color=$System.Color, Lifetime=5, Position Mode=::NewEnumerator0, Write Material Random=true)
   &SpawnBeam()
   &BeamWidth(Beam Width=1000)
   /SetVariables() {isSetParameters=true}
  ]
  particleUpdateScript [
   &ParticleState(Kill Particles When Lifetime Has Elapsed=false, Loop Particles Lifetime=true)
   &SolveForcesAndVelocity(Clamp Velocity=false, Limit Acceleration=false)
   &UpdateBeam()
  ]
  rendererClasses [&NiagaraRibbonRendererProperties]
  renderers [{rendererClass=&NiagaraRibbonRendererProperties}]
 }
 CPU_Particles {
  emitterUpdateScript [
   &EmitterState()
   &SpawnRate(SpawnRate=32, Use Spawn Probability=false)
  ]
  particleSpawnScript [
   &InitializeParticle(
    Color=$System.Color
    Lifetime Max=5
    Lifetime Min=2
    Lifetime Mode=::NewEnumerator1
    Sprite Size Mode=::NewEnumerator1
    Uniform Sprite Size Max=40
    Uniform Sprite Size Min=20
    UsePositionOffset=false
    Write Material Random=true
   )
   &self:SampleSpline(Spline:NiagaraDataInterfaceSpline=$System.Spline) {module=SampleSpline}
   &ShapeLocation(MaskSpawnGroup=false, Override Seed=false, Sphere Radius=25, UseOwnerScale=true)
   &ApplyInitialForces()
  ]
  particleUpdateScript [
   &ParticleState()
   &CurlNoiseForce(Bias Noise=false, Noise Frequency=10, Noise Strength=150, Pan Noise=true, Pan Noise Field=<0.25 1 0.25>)
   &self:SplinePhysics(
    ForceAround:NiagaraFloat=$System.ForceAround
    ForwardDistMax:NiagaraFloat=$System.ForwardDistMax
    ForwardDistMin:NiagaraFloat=$System.ForwardDistMin
    ForwardMax:NiagaraFloat=$System.ForwardMax
    ForwardMin:NiagaraFloat=$System.ForwardMin
    MagnetDistMax:NiagaraFloat=$System.MagnetDistMax
    MagnetDistMin:NiagaraFloat=$System.MagnetDistMin
    MagnetMax:NiagaraFloat=$System.MagnetMax
    MagnetMin:NiagaraFloat=$System.MagnetMin
    Spiral:NiagaraDataInterfaceCurve=$System.Spiral
    Spline:NiagaraDataInterfaceSpline=$System.Spline
   ) {module=SplinePhysics}
   &Drag(Use Linear Drag=true, Use Rotational Drag=true)
   &SolveForcesAndVelocity(Clamp Velocity=false, Limit Acceleration=false)
   &ScaleSpriteSizeBySpeed(Max Scale Factor=<2 16>, Min Scale Factor=<1 6>)
   &ScaleColor(
    Scale Alpha=&FloatFromCurve(FloatCurve=DataInterface() {propertyValues {Curve {
        defaultValue=1
        keys [
         {arriveTangent=9.0725, interpMode=RCIM_Cubic, leaveTangent=9.0725, tangentMode=RCTM_User, time=0, value=0}
         {arriveTangent=-2.5249, interpMode=RCIM_Cubic, leaveTangent=-2.5249, tangentMode=RCTM_User, time=0.1408, value=0.9847}
         {arriveTangent=-0.0405, interpMode=RCIM_Cubic, leaveTangent=-0.0405, tangentMode=RCTM_User, time=1, value=0}
        ]
       }}})
    Scale RGB=&Multiply_VectorByFloat(Vector=<1 1 1>, Float=$System.Multiplier)
    ScaleA=true
    ScaleRGB=true
   )
  ]
  rendererClasses [&NiagaraSpriteRendererProperties]
  renderers [{rendererClass=&NiagaraSpriteRendererProperties}]
 }
 GPU_Particles {
  simTarget=GPUComputeSim
  emitterUpdateScript [
   &EmitterState()
   &SpawnRate(SpawnRate=512, Use Spawn Probability=false)
  ]
  particleSpawnScript [
   &InitializeParticle(
    Color=$System.Color
    Lifetime Max=5
    Lifetime Min=2
    Lifetime Mode=::NewEnumerator1
    Sprite Size Mode=::NewEnumerator1
    UsePositionOffset=false
    Write Material Random=true
   )
   &self:SampleSpline(Spline:NiagaraDataInterfaceSpline=$System.Spline) {module=SampleSpline}
   &ShapeLocation(MaskSpawnGroup=false, Override Seed=false, Sphere Radius=500, UseOwnerScale=true)
   &ApplyInitialForces()
  ]
  particleUpdateScript [
   &ParticleState()
   &CurlNoiseForce(Bias Noise=false, Noise Frequency=10, Noise Strength=150, Pan Noise=true, Pan Noise Field=<0.25 1 0.25>)
   &self:SplinePhysics(
    ForceAround:NiagaraFloat=$System.ForceAround
    ForwardDistMax:NiagaraFloat=$System.ForwardDistMax
    ForwardDistMin:NiagaraFloat=$System.ForwardDistMin
    ForwardMax:NiagaraFloat=$System.ForwardMax
    ForwardMin:NiagaraFloat=$System.ForwardMin
    MagnetDistMax:NiagaraFloat=$System.MagnetDistMax
    MagnetDistMin:NiagaraFloat=$System.MagnetDistMin
    MagnetMax:NiagaraFloat=$System.MagnetMax
    MagnetMin:NiagaraFloat=$System.MagnetMin
    Spiral:NiagaraDataInterfaceCurve=$System.Spiral
    Spline:NiagaraDataInterfaceSpline=$System.Spline
   ) {module=SplinePhysics}
   &Drag(Use Linear Drag=true, Use Rotational Drag=true)
   &SolveForcesAndVelocity(Clamp Velocity=false, Limit Acceleration=false)
   &ScaleSpriteSizeBySpeed(Max Scale Factor=<2 16>, Min Scale Factor=<1 6>)
   &ScaleColor(
    Scale Alpha=&FloatFromCurve(FloatCurve=DataInterface() {propertyValues {Curve {
        defaultValue=1
        keys [
         {arriveTangent=9.0725, interpMode=RCIM_Cubic, leaveTangent=9.0725, tangentMode=RCTM_User, time=0, value=0}
         {arriveTangent=-2.5249, interpMode=RCIM_Cubic, leaveTangent=-2.5249, tangentMode=RCTM_User, time=0.1408, value=0.9847}
         {arriveTangent=-0.0405, interpMode=RCIM_Cubic, leaveTangent=-0.0405, tangentMode=RCTM_User, time=1, value=0}
        ]
       }}})
    Scale RGB=&Multiply_VectorByFloat(Vector=<1 1 1>, Float=$System.Multiplier)
    ScaleA=true
    ScaleRGB=true
   )
  ]
  rendererClasses [&NiagaraSpriteRendererProperties]
  renderers [{rendererClass=&NiagaraSpriteRendererProperties}]
 }
}
summary {
 systemName=NS_MatchModifier_Spline_Particles
 emitters [
  {emitterName=GPU_Particles, simTarget=GPUComputeSim, rendererClasses [&NiagaraSpriteRendererProperties]}
  {emitterName=CPU_Particles, rendererClasses [&NiagaraSpriteRendererProperties]}
  {emitterName=Beam_A, rendererClasses [&NiagaraRibbonRendererProperties]}
  {emitterName=Beam_B, rendererClasses [&NiagaraRibbonRendererProperties]}
 ]
 userVariables [User.Color {type=LinearColor, description="", defaultValue {struct=LinearColor, value=<rgba 1 0 0 1>}}]
}
