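# Niagara sprite effect of rising glow, streaks and motes, apparently one section of a match-modifier spline in Rift Anomalies, judging by its name.
# CenterGlow spawns tall sprites that grow while fading in and out.
# UpwardStreaks sends long, camera-offset streaks upward from an X-stretched sphere, shrinking with curve-driven color.
# SmallCircles lifts small motes from a similar shape using upward gravity, swirled by curl noise and vortex forces with drag.
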NiagaraSystem /RiftAnomaliesGameplay/Assets/FX/NS_MatchModifier_Spline_Section
dependencies {
 usedDataInterfaces [&NiagaraDataInterfaceCurve, &NiagaraDataInterfaceColorCurve, &NiagaraDataInterfaceVector2DCurve]
 usedDynamicInputs []
 usedModules [
  &SystemState
  &self:SystemSpawnScript.SystemScriptSource.SystemScriptGraph.NiagaraNodeAssignment_0.SetVariables_#1
 ]
}
emitters {
 CenterGlow {
  emitterUpdateScript [
   &EmitterState()
   &SpawnRate(SpawnRate=2, Use Spawn Probability=false)
  ]
  particleSpawnScript [&InitializeParticle(Position Offset=<0 0 250>, Sprite Size=<100 1000>, Sprite Size Mode=::NewEnumerator4, UsePositionOffset=true, Write Material Random=true)]
  particleUpdateScript [
   &ParticleState()
   &ScaleSpriteSize(Uniform Curve Sprite Scale=DataInterface() {propertyValues {Curve {keys [{tangentMode=RCTM_User, time=0, value=0}, <at 0.2 0.5>, {tangentMode=RCTM_User, time=1, value=1}]}}})
   &ScaleColor(
    Linear Color Curve=DataInterface() {propertyValues {
      AlphaCurve {keys [<at 0 0>,<at 0.1781 1>,<at 0.7078 1>,<at 0.9821 0>]}
      BlueCurve {keys [<at 0 1>,<at 1 1>]}
      GreenCurve {keys [<at 0 1>,<at 1 1>]}
      RedCurve {keys [<at 0 1>,<at 1 1>]}
     }}
    Scale Mode=::NewEnumerator2
   )
  ]
  rendererClasses [&NiagaraSpriteRendererProperties]
  renderers [{rendererClass=&NiagaraSpriteRendererProperties}]
 }
 SmallCircles {
  emitterUpdateScript [
   &EmitterState()
   &SpawnRate(SpawnRate=60, Use Spawn Probability=false)
  ]
  particleSpawnScript [
   &InitializeParticle(
    Lifetime Mode=::NewEnumerator1
    Position Offset=<0 0 30>
    Sprite Rotation Mode=::NewEnumerator2
    Sprite Size Mode=::NewEnumerator1
    Uniform Sprite Size Max=6
    Uniform Sprite Size Min=4
    UsePositionOffset=true
    Write Material Random=true
   )
   &ShapeLocation(
    MaskSpawnGroup=false
    Non Uniform Scale=<3 1 1>
    Offset=<0 0 40>
    Offset Mode=::NewEnumerator0
    Override Seed=false
    Sphere Radius=160
    Sphere Surface Distribution=1
    UseOwnerScale=true
   )
   &AddVelocity(Velocity=&RandomRangeVector(Minimum=<0 0 80>, Maximum=<0 0 120>)) {module=AddVelocity001}
  ]
  particleUpdateScript [
   &CurlNoiseForce(
    Bias Noise=false
    Noise Frequency=10
    Noise Strength=&RandomRangeFloat(Minimum=80, Maximum=155)
    Pan Noise=false
   )
   &Drag(Drag=2.5, Use Linear Drag=true, Use Rotational Drag=true)
   &VortexForce(
    Origin Pull Only On Vortex Axis=false
    UseOriginOffset=false
    Vortex Force Amount=&RandomRangeFloat(Minimum=250, Maximum=600)
   )
   &GravityForce(Gravity=<0 0 400>)
   &SolveForcesAndVelocity(Clamp Velocity=false, Limit Acceleration=false)
   &ParticleState()
   &ScaleSpriteSize(Uniform Curve Sprite Scale=DataInterface() {propertyValues {Curve {keys [
        {interpMode=RCIM_Cubic, tangentMode=RCTM_User, time=0, value=0.55}
        {interpMode=RCIM_Cubic, tangentMode=RCTM_User, time=0.5, value=1}
        {arriveTangent=-3.7198, tangentMode=RCTM_User, time=1, value=0}
       ]}}})
  ]
  rendererClasses [&NiagaraSpriteRendererProperties]
  renderers [{rendererClass=&NiagaraSpriteRendererProperties}]
 }
 UpwardStreaks {
  emitterUpdateScript [
   &EmitterState()
   &SpawnRate(SpawnRate=18, Use Spawn Probability=false)
  ]
  particleSpawnScript [
   &InitializeParticle(
    Lifetime=&RandomRangeFloat(Minimum=1, Maximum=2)
    Position=<0 0 200>
    Position Mode=::NewEnumerator1
    Sprite Size=&RandomRangeVector2D(Minimum=<10 2500>, Maximum=<35 4500>)
    Sprite Size Mode=::NewEnumerator4
    UsePositionOffset=false
    Write Material Random=true
   )
   &DynamicMaterialParameters(
    Index 0 Param 1=1
    Index 0 Param 2=4
    Index 0 Param 3=1
    Index 0 Param 4=1
    Module.Param0WriteEnabled=true
    Module.Param1WriteEnabled=true
    Module.Param2WriteEnabled=true
    Module.Param3WriteEnabled=true
   )
   &ShapeLocation(
    MaskSpawnGroup=false
    Non Uniform Scale=<3 1 1>
    Offset=<0 0 200>
    Offset Coordinate Space=::Local
    Offset Mode=::NewEnumerator0
    Override Seed=false
    Shape Origin=$Engine.Emitter.SimulationPosition
    Sphere Radius=180
    Sphere Surface Distribution=1
    UseOwnerScale=true
   )
   &AddVelocity(Velocity=<0 0 1200>)
  ]
  particleUpdateScript [
   &ParticleState()
   &SolveForcesAndVelocity(Clamp Velocity=false, Limit Acceleration=false)
   &CameraOffset(Camera Offset Amount=27.1877)
   &self:ColorSet_01(ScaleCurve:NiagaraFloat=&FloatFromCurve(
     FloatCurve=DataInterface() {propertyValues {Curve {keys [
         {arriveTangent=0.0641, interpMode=RCIM_Cubic, leaveTangent=0.0641, tangentMode=RCTM_User, time=0, value=0}
         {arriveTangent=-2.5249, interpMode=RCIM_Cubic, leaveTangent=-2.5249, tangentMode=RCTM_User, time=0.1, value=0.975}
         {arriveTangent=-0.0405, interpMode=RCIM_Cubic, leaveTangent=-0.0405, tangentMode=RCTM_User, time=1, value=0}
        ]}}}
     Scale Curve=0.684
    )) {module=ColorSet_01}
   &ScaleSpriteSize(
    Scale Factor=&Vector2DFromCurve(
     Vector2Curve=DataInterface() {propertyValues {
       XCurve {keys [<at 0 2.15 RCIM_Cubic>,<at 1 0 RCIM_Cubic>]}
       YCurve {keys [{arriveTangent=-2.543, interpMode=RCIM_Cubic, leaveTangent=-2.543, tangentMode=RCTM_User, time=0, value=1}, <at 1 -0.05 RCIM_Cubic>]}
      }}
     ScaleCurve=<1 1>
    )
    Scale Sprite Size Mode=::NewEnumerator2
   )
  ]
  rendererClasses [&NiagaraSpriteRendererProperties]
  renderers [{rendererClass=&NiagaraSpriteRendererProperties}]
 }
}
summary {
 systemName=NS_MatchModifier_Spline_Section
 emitters [
  {emitterName=CenterGlow, rendererClasses [&NiagaraSpriteRendererProperties]}
  {emitterName=UpwardStreaks, rendererClasses [&NiagaraSpriteRendererProperties]}
  {emitterName=SmallCircles, rendererClasses [&NiagaraSpriteRendererProperties]}
 ]
 userVariables []
}
