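# Niagara beacon effect for a rift anomaly: a tall vertical light beam with rising chromatic rays and sparks.
# Beam_Main has two mesh renderers plus thinner scale/color attributes, and fades in on the system Timeline.
# ChromaRay and GPU Sparks spawn sprites in cylinders that drift upward and fade.
# LOD-distance remaps enlarge elements and strengthen the beam with distance; User.FX_Intensity drives brightness, while time-of-day and platform also adjust the beam.
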NiagaraSystem /RiftAnomaliesGameplay/Assets/FX/NS_RiftAnomaly_Beacon
dependencies {
 usedDataInterfaces [
  &NiagaraDataInterfaceMeshRendererInfo
  &NiagaraDataInterfaceCurve
  &NiagaraDataInterfacePlatformSet
  &NiagaraDataInterfaceVectorCurve
 ]
 usedDynamicInputs [/Game/Effects/Niagara/DynamicInputs/State/FN_IsExecutionStateActive, &InvertBool]
 usedModules [&SystemState, &Timeline]
}
emitters {
 Beam_Main {
  emitterUpdateScript [
   &EmitterState(Life Cycle Mode=::NewEnumerator1, Loop Behavior=::NewEnumerator1, Loop Duration Mode=::NewEnumerator1, UseLoopDelay=false)
   &SpawnBurst_Instantaneous(Use Spawn Probability=false, UseLoopCountLimit=false)
   &SpawnRate(Use Spawn Probability=false) {enabled=false}
  ]
  particleSpawnScript [
   &InitializeParticle(
    Color=&Multiply_LinearColorByFloat(Linear Color=<rgba 0.5833 0 1 1>, Float=$User.FX_Intensity)
    Mesh Scale=<2.5 2.5 200>
    Mesh Scale Mode=::NewEnumerator4
    Position Offset=<0 0 -100>
    Sprite Size Mode=::NewEnumerator3
    Uniform Sprite Size=200
    UsePositionOffset=true
    Write Material Random=true
   )
   &CameraOffset(Camera Offset Amount=&FN_PickFloatDependOnPlatform(TargetPlatformActive=-100, TargetPlatformNoActive=-150))
   &InitialMeshOrientation(Mesh Axis To Orient=<0 0 1>, Mesh Orientation Mode=::NewEnumerator2, Orientation Vector=<0 0 1>, Use Rotation Vector=false)
  ]
  particleUpdateScript [
   &ParticleState(
    Kill Particles When Lifetime Has Elapsed=false
    Let Infinitely Lived Particles Die When Emitter Deactivates=true
    Loop Particles Lifetime=true
   )
   &ScaleMeshSize(Scale Factor=&VectorFromCurve(
     VectorCurve=DataInterface() {propertyValues {
       XCurve {keys [{arriveTangent=1.9285, interpMode=RCIM_Cubic, leaveTangent=1.9285, tangentMode=RCTM_User, time=0, value=0.8}, <at 1 1.5 RCIM_Cubic>]}
       YCurve {keys [{arriveTangent=1.9285, interpMode=RCIM_Cubic, leaveTangent=1.9285, tangentMode=RCTM_User, time=0, value=0.8}, <at 1 1.5 RCIM_Cubic>]}
       ZCurve {keys [<at 0 1>,<at 1 1>]}
      }}
     CurveIndex=$System.Timeline.NormalizedTime
     Scale Curve=&VectorFromFloat(Value=&FloatFromCurve(
       FloatCurve=DataInterface() {propertyValues {Curve {
          defaultValue=0
          keys [
           {arriveTangent=1.7109, interpMode=RCIM_Cubic, leaveTangent=1.7109, tangentMode=RCTM_User, time=0, value=0}
           {interpMode=RCIM_Cubic, tangentMode=RCTM_User, time=1, value=1}
          ]
         }}}
       CurveIndex=&FN_LOD_Distance_Remap(Start Distance=0, End Distance=7500, Minimum Output=0, Maximum Output=1, Clamp Output=true)
       Scale Curve=2
      ))
    ))
   &ScaleColor(
    Scale Alpha=&FloatFromCurve(
     FloatCurve=DataInterface() {propertyValues {Curve {keys [
         <at 0 0>
         {arriveTangent=1.25, time=0.5, value=1}
         {arriveTangent=-0.9805, interpMode=RCIM_Cubic, leaveTangent=-0.9805, tangentMode=RCTM_User, time=0.5, value=1}
        ]}}}
     CurveIndex=$System.Timeline.Time
     Scale Curve=&Multiply_Float(
      A=&FN_LOD_Distance_Remap(Start Distance=50, End Distance=5000, Minimum Output=0, Maximum Output=4, Clamp Output=true)
      B=&Multiply_Float(
       A=&FNTimeOfDayScalar(Night Scale=0.8, Morning Scale=1, Evening Scale=1)
       B=&FN_PickFloatDependOnPlatform(TargetPlatformActive=1, TargetPlatformNoActive=0.6)
      )
     )
    )
    ScaleA=true
    ScaleRGB=true
   )
   /SetVariables(
    Particles.Color_Thinner:LinearColor=&MultiplyLinearColor(
     Linear Color=$Particles.Color
     Scale Factor=&MakeVector4FromVectorAndFloat(
      "Vector (RGB)"=&MakeVectorFromLinearColorRGB(Color=<rgba 0.498 0.1693 2 1.25>)
      "Float (Alpha)"=&Multiply_Float(
       A=&FloatFromCurve(
        FloatCurve=DataInterface() {propertyValues {Curve {defaultValue=1, keys [{tangentMode=RCTM_User, time=0, value=0}, {tangentMode=RCTM_User, time=1, value=1}]}}}
        CurveIndex=$System.Timeline.NormalizedTime
       )
       B=&Multiply_Float(
        A=1.5
        B=&FN_LOD_Distance_Remap(Start Distance=50, End Distance=5000, Minimum Output=0, Maximum Output=2, Clamp Output=true)
       )
      )
     )
    )
    Particles.Scale_Thinner:Vector3f=&Multiply_Vector(A=$Particles.Scale, B=<0.35 0.35 0.75>)
   ) {isSetParameters=true}
   &ScaleColor(
    Scale Alpha=&FloatFromCurve(
     FloatCurve=DataInterface() {propertyValues {Curve {defaultValue=1, keys [{tangentMode=RCTM_User, time=0, value=0}, {tangentMode=RCTM_User, time=1, value=1}]}}}
     CurveIndex=$System.Timeline.NormalizedTime
    )
    ScaleA=true
    ScaleRGB=true
   ) {enabled=false, module=ScaleColor002}
   &ScaleColor(
    Scale Alpha=&Multiply_Float(
     A=&Multiply_Float(
      A=&FloatFromCurve(
       FloatCurve=DataInterface() {propertyValues {Curve {defaultValue=0, keys [{tangentMode=RCTM_User, time=0, value=0}, {tangentMode=RCTM_User, time=1, value=1}]}}}
       CurveIndex=$System.Timeline.Time
       Scale Curve=$User.FX_Intensity
      )
      B=1
     )
     B=1
    )
    ScaleA=true
    ScaleRGB=true
   ) {module=ScaleColor003}
  ]
  rendererClasses [&NiagaraMeshRendererProperties]
  renderers [
   {rendererClass=&NiagaraMeshRendererProperties, rendererIndex=0}
   {rendererClass=&NiagaraMeshRendererProperties, rendererIndex=1}
  ]
 }
 ChromaRay {
  emitterUpdateScript [
   &EmitterState(Life Cycle Mode=::NewEnumerator1, Loop Behavior=::NewEnumerator1, Loop Delay=0.4, Loop Duration Mode=::NewEnumerator1, UseLoopDelay=false)
   &SpawnRate(SpawnRate=9, Use Spawn Probability=false)
  ]
  particleSpawnScript [
   &InitializeParticle(
    Color Maximum=<rgba 0.7271 0.0847 1 0.4>
    Color Minimum=<rgba 0.0084 0.4522 1 0.4>
    Color Mode=::NewEnumerator2
    Lifetime Max=1.5
    Lifetime Min=2
    Lifetime Mode=::NewEnumerator1
    Sprite Size Max=<150 750>
    Sprite Size Min=<100 400>
    Sprite Size Mode=::NewEnumerator2
    Sprite UV Mode=::NewEnumerator3
    UsePositionOffset=false
    Write Material Random=true
   )
   &ShapeLocation(
    Cylinder Height=1500
    Cylinder Height Midpoint=0
    Cylinder Radius=60
    MaskSpawnGroup=false
    Offset=<0 0 500>
    Offset Mode=::NewEnumerator0
    Override Seed=false
    Shape Primitive=::NewEnumerator1
    Surface Only=false
    UseOwnerScale=true
   )
   &AddVelocity(Velocity=<0 0 1>, Velocity Speed Scale=&RandomRangeFloat(Minimum=25, Maximum=100)) {module=AddVelocity001}
   &DynamicMaterialParameters(
    Index 0 Param 1=&RandomRangeFloat(Minimum=0, Maximum=1)
    Index 0 Param 2=&RandomRangeFloat(Minimum=0, Maximum=1)
    Module.Param0WriteEnabled=true
    Module.Param1WriteEnabled=true
    Module.Param2WriteEnabled=true
    Module.Param3WriteEnabled=true
   ) {module=DynamicMaterialParameters001}
  ]
  particleUpdateScript [
   &ParticleState()
   &ScaleSpriteSize(
    Uniform Curve Scale=&FloatFromCurve(
     FloatCurve=DataInterface() {propertyValues {Curve {
        defaultValue=0
        keys [
         {arriveTangent=1.7109, interpMode=RCIM_Cubic, leaveTangent=1.7109, tangentMode=RCTM_User, time=0, value=0}
         {interpMode=RCIM_Cubic, tangentMode=RCTM_User, time=1, value=1}
        ]
       }}}
     CurveIndex=&FN_LOD_Distance_Remap(Start Distance=0, End Distance=7500, Minimum Output=0, Maximum Output=1, Clamp Output=true)
     Scale Curve=5
    )
    Uniform Curve Sprite Scale=DataInterface() {propertyValues {Curve {keys [{tangentMode=RCTM_User, time=0, value=1}, {tangentMode=RCTM_User, time=1, value=1.2}]}}}
   )
   &GravityForce(Gravity=&RandomRangeVector(Minimum=<0 0 111>, Maximum=<0 0 22>))
   &Drag(Drag=0.5, Use Linear Drag=true, Use Rotational Drag=true)
   &SolveForcesAndVelocity(Clamp Velocity=false, Limit Acceleration=false)
   &ScaleColor(
    Scale Alpha=&Multiply_Float(
     A=&Multiply_Float(
      A=&FloatFromCurve(
       FloatCurve=DataInterface() {propertyValues {Curve {keys [
           {interpMode=RCIM_Cubic, tangentMode=RCTM_User, time=0, value=0}
           {interpMode=RCIM_Cubic, tangentMode=RCTM_User, time=0.5, value=1}
           {tangentMode=RCTM_User, time=1, value=0}
          ]}}}
       Scale Curve=$User.FX_Intensity
      )
      B=1
     )
     B=1
    )
    Scale RGB=&VectorFromFloat(Value=$User.FX_Intensity)
    ScaleA=true
    ScaleRGB=true
   )
  ]
  rendererClasses [&NiagaraSpriteRendererProperties]
  renderers [{rendererClass=&NiagaraSpriteRendererProperties}]
 }
 Sparks {
  simTarget=GPUComputeSim
  emitterUpdateScript [
   &EmitterState(Life Cycle Mode=::NewEnumerator1, Loop Behavior=::NewEnumerator1, Loop Duration Mode=::NewEnumerator1, UseLoopDelay=false)
   &SpawnRate(SpawnRate=15, Use Spawn Probability=false)
  ]
  particleSpawnScript [
   &InitializeParticle(
    Color=&Multiply_LinearColorByFloat(Linear Color=<rgba 0.3006 1.2717 2 0.2>, Float=$User.FX_Intensity)
    Lifetime Max=3
    Lifetime Min=2
    Lifetime Mode=::NewEnumerator1
    Mesh Scale Max=<0.38 0.28 0.5>
    Mesh Scale Min=<0.05 0.05 0.05>
    Mesh Scale Mode=::NewEnumerator2
    Sprite Rotation Mode=::NewEnumerator2
    Sprite Size Mode=::NewEnumerator1
    Uniform Sprite Size Max=30
    Uniform Sprite Size Min=15
    UsePositionOffset=true
    Write Material Random=true
   )
   &ShapeLocation(
    Cylinder Height=1500
    Cylinder Radius=250
    MaskSpawnGroup=false
    Non Uniform Scale=<1 1 2>
    Offset=<0 0 300>
    Offset Mode=::NewEnumerator0
    Override Seed=false
    Shape Primitive=::NewEnumerator1
    Surface Only=false
    UseOwnerScale=true
   )
   &AddVelocity(Velocity=<0 0 1>, Velocity Speed Scale=&RandomRangeFloat(Minimum=10, Maximum=25)) {module=AddVelocity001}
  ]
  particleUpdateScript [
   &ParticleState()
   &ScaleSpriteSize(Uniform Curve Scale=&FloatFromCurve(
     FloatCurve=DataInterface() {propertyValues {Curve {
        defaultValue=0
        keys [
         {arriveTangent=1.7109, interpMode=RCIM_Cubic, leaveTangent=1.7109, tangentMode=RCTM_User, time=0, value=0}
         {interpMode=RCIM_Cubic, tangentMode=RCTM_User, time=1, value=1}
        ]
       }}}
     CurveIndex=&FN_LOD_Distance_Remap(Start Distance=0, End Distance=7500, Minimum Output=0, Maximum Output=1, Clamp Output=true)
     Scale Curve=5
    ))
   &GravityForce(Gravity=&RandomRangeVector(Minimum=<0 0 111>, Maximum=<0 0 22>))
   &SolveForcesAndVelocity(Clamp Velocity=false, Limit Acceleration=false)
   &ScaleColor(
    Scale Alpha=&FloatFromCurve(FloatCurve=DataInterface() {propertyValues {Curve {
        defaultValue=0
        keys [
         {interpMode=RCIM_Cubic, tangentMode=RCTM_User, time=0, value=0}
         {interpMode=RCIM_Cubic, tangentMode=RCTM_User, time=0.5, value=1}
         {tangentMode=RCTM_User, time=1, value=0}
        ]
       }}})
    ScaleA=true
    ScaleRGB=true
   )
  ]
  rendererClasses [&NiagaraSpriteRendererProperties]
  renderers [{rendererClass=&NiagaraSpriteRendererProperties}]
 }
}
summary {
 systemName=NS_RiftAnomaly_Beacon
 emitters [
  {emitterName=Sparks, simTarget=GPUComputeSim, rendererClasses [&NiagaraSpriteRendererProperties]}
  {emitterName=Beam_Main, rendererClasses [&NiagaraMeshRendererProperties]}
  {emitterName=ChromaRay, rendererClasses [&NiagaraSpriteRendererProperties]}
 ]
 userVariables [User.FX_Intensity {type=NiagaraFloat, defaultValue=1, description=""}]
}
